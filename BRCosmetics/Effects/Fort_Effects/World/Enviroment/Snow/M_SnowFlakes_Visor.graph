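# Unlit translucent sprite particle material for snowflakes, tinted by particle color through an emissive adjustment function.
# Pixel depth and particle speed blend between a blurred snow texture and its product with a sharp one.
# Particle random picks an RGB channel as the flake's opacity, scaled by particle alpha.
# A near-camera fade lowers opacity and collapses vertices toward the particle center; the mobile path skips both.
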Material /BRCosmetics/Effects/Fort_Effects/World/Enviroment/Snow/M_SnowFlakes_Visor
graph {
 properties {
  bUsedWithParticleSprites=true
  blendMode=BLEND_Translucent
  pixelDepthOffsetMode=PDOM_Legacy
  shadingModel=MSM_Unlit
  translucencyLightingMode=TLM_VolumetricPerVertexDirectional
  usedShadingModels=Unlit
 }
 nodes [
  n1 = ParticleColor()
  MaterialFunctionCall(Emissive=@n1.RGB) {materialFunction=&MF_FX_EmissiveAdjustments}
  n2 = ParticleRandom()
  n3 = TextureSample() {texture=/Game/Weapons/WeaponSkins/Elemental/Textures/T_Snow_01_blur}
  n4 = PixelDepth()
  n5 = Constant()
  n6 = ScalarParameter() {defaultValue=800, parameterName=NoBlurRange_Centered}
  n7 = SphereMask(A=PixelDepth().R, B=Constant(), Radius=@n6) {hardnessPercent=10}
  n8 = Abs(None=Subtract(A=SphereMask(A=@n4.R, B=@n5, Radius=@n6) {hardnessPercent=10}, B=OneMinus(None=@n7)))
  n9 = Divide(A=ParticleSpeed(), B=ScalarParameter() {defaultValue=500, parameterName=SpeedBlurThreshold}) {constB=400}
  n10 = LinearInterpolate(
   A=Multiply(A=TextureSample() {texture=/Game/Weapons/WeaponSkins/Elemental/Textures/T_Snow_01}.RGB, B=@n3.RGB)
   B=@n3.RGB
   Alpha=Clamp(None=Subtract(A=@n8, B=@n9))
  )
  n11 = ScalarParameter() {defaultValue=100, parameterName=ScaleDown_Start}
  n12 = Clamp(None=Subtract(
    A=MaterialFunctionCall(Distance=@n11) {materialFunction=&Near_Camera_Fade}.Result
    B=ScalarParameter() {parameterName=ScaleDown_End_Normalized}
   ))
  Multiply(
   A=Multiply(
    A=@n1.A
    B=MaterialFunctionCall!(Random Picker=@n2, Texture Input=@n10, Option 1, Option 2, Option 3) {materialFunction=/Game/Effects/Fort_Effects/Materials/_Functions/MF_RGB_Random_01}.Result
   )
   B=ShadingPathSwitch(Default=@n12, Mobile=Constant() {r=1}) {desc=Migrated from FeatureLevelSwitch}
  )
  n13 = ParticlePositionWS()
  n14 = WorldPosition()
  ShadingPathSwitch(Default=LinearInterpolate(A=Subtract(A=@n13, B=@n14.XYZ), B=Constant(), Alpha=@n12), Mobile=Constant()) {desc=Migrated from FeatureLevelSwitch}
 ]
}
